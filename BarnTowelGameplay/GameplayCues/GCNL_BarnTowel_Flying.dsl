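; Looping gameplay cue that runs trail FX, audio and camera while a player flies with BarnTowel.
; On start it enables the flying camera mode and modifier, audio bus mix and non-teammate sound indicator, and watches fuel and crash-tag changes.
; Repeating timers feed speed, anim-layer and height values into FX and audio, and swap vertical-movement sounds.
; Removal undoes these and conditionally plays a landing sound; console variables gate extra camera-cleanup hooks.
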(event EventOnLoopingStart(EitherVFX-preferthisversion)
  (params (MyTarget :type Actor%20Object%20Reference) (Parameters :type Gameplay%20Cue%20Parameters%20Structure%20%28by%20ref%29) (ParticleComponents :type Array%20of%20FXSystem%20Component%20Object%20References) (AudioComponents :type Array%20of%20Audio%20Component%20Object%20References))
  (bind _fort_player_controller (Variables|Default|GetFort%20Player%20Controller))
  (|CallParentFunction MyTarget Parameters ParticleComponents AudioComponents)
  (Variables|SpeedSmoothing|SetPreviousSpeed 0.0)
  (bind _asfort_player_pawn_athena (Utilities|Casting|CastToFortPlayerPawnAthena MyTarget))
  (Variables|Default|SetPlayerPawn _asfort_player_pawn_athena)
  (bind _asbarn_towel_player_anim_bp (Utilities|Casting|CastToBarnTowel_Player_AnimBP (Components|SkeletalMesh|AnimationBlueprintLinking|GetLinkedAnimLayerInstancebyClass (Variables|/Script/Engine.Character|GetMesh (Variables|Default|GetPlayerPawn)) "/BarnTowelGameplay/Assets/Animation/BarnTowel_Player_AnimBP.BarnTowel_Player_AnimBP_C"))
    (:then
      (Variables|Transient|SetPlayerAnimBP _asbarn_towel_player_anim_bp))
    (:CastFailed
      _))
  (bind (_asynctaskproxy _newvalue) (Ability|Tasks|WaitForSimulatedAttributeChanged (Variables|Default|GetPlayerPawn) "(AttributeName=\"LocalFuel\",Attribute=/Script/FlyingCodeRuntime.FlyingAttributeSet:LocalFuel,AttributeOwner=\"/Script/CoreUObject.Class'/Script/FlyingCodeRuntime.FlyingAttributeSet'\")")
    (:then
      (Utilities|FlowControl|Sequence
        (:then_0
          (bind (_asynctaskproxy_2 _tagcount) (Ability|Tasks|WaitGameplayTagCountChangedOnActor (Variables|Default|GetPlayerPawn) (Variables|Default|GetTag_Crashing))
            (:then
              (Variables|AudioVisualizer|SetAction_WaitCrashTagChanged _asynctaskproxy_2)
              (Utilities|FlowControl|Sequence
                (:then_0
                  (bind _asb_weapon_barn_towel (Utilities|Casting|CastToB_Weapon_BarnTowel (Variables|/Script/FortniteGame.FortPawn|GetCurrentWeapon (Variables|Default|GetPlayerPawn))))
                  (Variables|TransientComponents|SetBarnTowelWeapon _asb_weapon_barn_towel)
                  (Variables|TransientComponents|SetAudioComponents AudioComponents)
                  (Variables|TransientComponents|SetFlyingTrailFX ParticleComponents)
                  (CallFunction|InitializeEffectsParms))
                (:then_1
                  (Variables|AudioVisualizer|SetIsCrashing (GameplayTags|HasMatchingGameplayTag (Variables|Default|GetPlayerPawn) (Variables|Default|GetTag_Crashing))))))
            (:TagCountChanged
              ;[1 1 1] AV Indicator
              (Variables|AudioVisualizer|SetIsCrashing (> _tagcount 0))))
          (if (and (not (Team|OnLocalPlayersTeam (Variables|Default|GetPlayerPawn) (Variables|Default|GetPlayerPawn))) (not (Variables|AudioVisualizer|GetIsCrashing)))
            (SoundIndicator|StartTargetedSoundIndicatorLooping (Variables|Default|GetFortSoundIndicator) (Variables|Default|GetPlayerPawn) "/BarnTowelGameplay/Assets/Sounds/_various/MSS_BarnTowel_FlyingAir_3P.MSS_BarnTowel_FlyingAir_3P" 1.0 (Variables|/Script/Engine.Pawn|GetPlayerState (Variables|Default|GetPlayerPawn)))))
        (:then_1
          (CallFunction|SetCameraEnabled (Variables|Default|GetPlayerPawn) true)
          (Utilities|FlowControl|Sequence
            (:then_0
              (bind _returnvalue (Development|GetConsoleVariableBoolValue "Fort.Rollback.ForceCleanWingsCameraWhenSwitchViewTarget"))
              (if _returnvalue
                (Utilities|IsValid _fort_player_controller
                  (:Is%20Valid
                    (EventDispatchers|BindEventtoOnViewTargetChanged _fort_player_controller (EventDispatchers|CreateEvent :!event Custom|HandleViewTargetChanged))))))
            (:then_1
              (bind _returnvalue_2 (Development|GetConsoleVariableBoolValue "Fort.Rollback.ForceCleanWingsCameraWhenFlyingEnds"))
              (if _returnvalue_2
                (bind _asynctaskproxy_3 (Ability|Tasks|WaitGameplayTagRemoveFromActor (Variables|Default|GetPlayerPawn) (Variables|Default|GetTag_FlyingMovementMode))
                  (:then
                    (Variables|Default|SetAsync%20Action_FlyMode _asynctaskproxy_3))
                  (:Removed
                    (CallFunction|SetCameraEnabled (Variables|Default|GetPlayerPawn))))))))))
    (:Changed
      (CallFunction|OnFuelChanged _newvalue (Variables|Transient|GetPreviousFuelValue)))))

(event EventOnRemoval(EitherVFX-preferthisversion)
  (params (MyTarget :type Actor%20Object%20Reference) (Parameters :type Gameplay%20Cue%20Parameters%20Structure%20%28by%20ref%29) (ParticleComponents :type Array%20of%20FXSystem%20Component%20Object%20References) (AudioComponents :type Array%20of%20Audio%20Component%20Object%20References) (BurstCameraShakeInstance :type Camera%20Shake%20Base%20Object%20Reference) (BurstDecalInstance :type Decal%20Actor%20Object%20Reference))
  (bind _fort_player_controller (Variables|Default|GetFort%20Player%20Controller))
  (Utilities|FlowControl|Sequence
    (:then_0
      (CallFunction|SetCameraEnabled MyTarget)
      (Variables|Default|SetFortPlayerCameraBase 0)
      (Utilities|IsValid _fort_player_controller
        (:Is%20Valid
          (EventDispatchers|UnbindEventfromOnViewTargetChanged _fort_player_controller (EventDispatchers|CreateEvent :!event Custom|HandleViewTargetChanged))
          (Variables|Default|SetFort%20Player%20Controller 0))))
    (:then_1
      (Utilities|Time|ClearandInvalidateTimerbyHandle MyTarget (Variables|Default|GetTimer_IsFlapingCheck))
      (Utilities|Time|ClearandInvalidateTimerbyHandle MyTarget (Variables|Default|GetTimer_VerticalAudio))
      (CallFunction|HandleControlBusMix false (Variables|Default|GetPlayerPawn))
      (SoundIndicator|StopSoundIndicatorLooping (Variables|Default|GetFortSoundIndicator))
      (Ability|Async|EndAction (Variables|AudioVisualizer|GetAction_WaitCrashTagChanged)))
    (:then_2
      ;[0.008285 0.01824 0.15 0.5] Handle Default "Landing" Sound if Not Out of Energy or Crashing
      (if (not (or (Variables|/Script/BarnTowelRuntime.BarnTowelLayerAnimInstance|GetbNoFuel (Variables|Transient|GetPlayerAnimBP)) (Variables|/Script/BarnTowelRuntime.BarnTowelLayerAnimInstance|GetbThrusting (Variables|Transient|GetPlayerAnimBP))))
        (Audio|SpawnSoundAttached (Variables|Sound|GetLandingSound) (Variables|/Script/Engine.Actor|GetRootComponent MyTarget))))))

(event Custom|HandleViewTargetChanged
  (params (PlayerController :type Fort%20Player%20Controller%20Object%20Reference) (OldViewTarget :type Actor%20Object%20Reference) (NewViewTarget :type Actor%20Object%20Reference))
  (CallFunction|SetCameraEnabled OldViewTarget))

(fn DebugParameters
  (Development|PrintString 0 (Utilities|String|Append "Speed: " (Utilities|String|ToString(Float) (Variables|SpeedSmoothing|GetPreviousSpeed)) " Roll: " (Utilities|String|ToString(Float) (Variables|/Script/BarnTowelRuntime.BarnTowelLayerAnimInstance|GetGlideRollCombined (Variables|Transient|GetPlayerAnimBP))) " Pitch: " (Utilities|String|ToString(Float) (Variables|/Script/BarnTowelRuntime.BarnTowelLayerAnimInstance|GetGlidePitchAlpha (Variables|Transient|GetPlayerAnimBP))) "Is Boosting" (Utilities|String|ToString(Boolean) (Variables|/Script/BarnTowelRuntime.BarnTowelLayerAnimInstance|GetbIsCrashing (Variables|Transient|GetPlayerAnimBP)))) true true "(R=0.000000,G=0.660000,B=1.000000,A=1.000000)" 5.0 "Pitch"))

(fn GetClampedSpeed
  (params (ActorComponent :type Actor%20Component%20Object%20Reference))
  (return (Math|Float|Clamp(Float) (Math|Vector|VectorLength (Transformation|GetVelocity (Components|GetOwner ActorComponent))) 4e+02 2.4e+03)))

(fn HandleControlBusMix
  (params (Apply :type Boolean) (PlayerPawn :type Fort%20Player%20Pawn%20Object%20Reference))
  (bind _playerpawn (Variables|Default|GetPlayerPawn))
  (if (Pawn|IsLocallyViewed PlayerPawn)
    (if (Variables|GetApply)
      (Audio|Modulation|ActivateControlBusMix _playerpawn (Variables|Sound|GetControlBusMix))
      (else
        (Audio|Modulation|DeactivateControlBusMix _playerpawn (Variables|Sound|GetControlBusMix))))))

(fn HandleVerticalMovementAudio
  (bind _barntowelweapon (Variables|TransientComponents|GetBarnTowelWeapon))
  (bind _flyingstatus (Variables|/Script/BarnTowelRuntime.FortWeaponBarnTowel|GetFlyingStatus _barntowelweapon))
  (bind _returnvalue (Math|Conversions|ToInteger(Byte) _flyingstatus))
  (bind _verticalupsound (Variables|Sound|GetVerticalUpSound))
  (bind _verticaldownsound (Variables|Sound|GetVerticalDownSound))
  (bind _returnvalue_2 (Utilities|Select :Index _returnvalue :Option%2011 _verticaldownsound :Option%2014 _verticalupsound))
  (bind _barntowelweapon_2 (Variables|TransientComponents|GetBarnTowelWeapon))
  (bind _verticalaudio (Variables|/BarnTowelGameplay/Gameplay/B_Weapon_BarnTowel.B_Weapon_BarnTowel_C|GetVerticalAudio _barntowelweapon_2))
  (Utilities|FlowControl|Sequence
    (:then_0
      (if (!= (Variables|Sound|GetPreviousFlyingStatus) (Math|Conversions|ToInteger(Byte) (Variables|/Script/BarnTowelRuntime.FortWeaponBarnTowel|GetFlyingStatus (Variables|TransientComponents|GetBarnTowelWeapon))))
        (Utilities|IsValid _returnvalue_2
          (:Is%20Valid
            (if (Audio|Components|Audio|IsPlaying _verticalaudio)
              (Audio|Components|Audio|FadeOut _verticalaudio 0.2 0.0 "Linear" :!class /Script/Engine.AudioComponent))
            (Audio|Components|Audio|SetSound _verticalaudio _returnvalue_2)
            (Audio|Components|Audio|Play _verticalaudio 0.0)))))
    (:then_1
      (Variables|Sound|SetPreviousFlyingStatus (Math|Conversions|ToInteger(Byte) (Variables|/Script/BarnTowelRuntime.FortWeaponBarnTowel|GetFlyingStatus (Variables|TransientComponents|GetBarnTowelWeapon)))))))

(fn InitializeEffectsParms
  (bind _playerpawn (Variables|Default|GetPlayerPawn))
  (bind (_bsuccessfullyfoundattribute _returnvalue) (Ability|Attribute|GetFloatAttribute _playerpawn "(AttributeName=\"LocalFuel\",Attribute=/Script/FlyingCodeRuntime.FlyingAttributeSet:LocalFuel,AttributeOwner=\"/Script/CoreUObject.Class'/Script/FlyingCodeRuntime.FlyingAttributeSet'\")"))
  (CallFunction|HandleControlBusMix true (Variables|Default|GetPlayerPawn))
  (CallFunction|SetFXFloatParameter "EnergyRemaining" (/ _returnvalue 1e+02))
  (Variables|Transient|SetPreviousFuelValue _returnvalue)
  (bind _returnvalue_2 (Utilities|Time|SetTimerbyEvent (EventDispatchers|CreateEvent :!event UpdateFXandAudioParams) 0.1 true))
  (Variables|Default|SetTimer_IsFlapingCheck _returnvalue_2)
  (bind _returnvalue_3 (Utilities|Time|SetTimerbyEvent (EventDispatchers|CreateEvent :!event HandleVerticalMovementAudio) 0.1 true))
  (Variables|Default|SetTimer_VerticalAudio _returnvalue_3)
  (for _array_element (Variables|TransientComponents|GetFlyingTrailFX)
    (Effects|Components|ParticleSystem|SetBoolParameter _array_element "IsLocalPlayer" (Pawn|IsLocallyControlled (Variables|Default|GetPlayerPawn))))
  (for _array_element_2 (Variables|TransientComponents|GetAudioComponents)
    (Audio|Parameter|SetBooleanParameter _array_element_2 "isLocalPlayer" (Pawn|IsLocallyViewed (Variables|Default|GetPlayerPawn)) :!class /Script/AudioExtensions.AudioParameterControllerInterface)))

(fn OnFuelChanged
  (params (NewValue :type Float%20%28double-precision%29) (OldValue :type Float%20%28double-precision%29))
  (CallFunction|SetFXFloatParameter "EnergyRemaining" (/ NewValue 1e+02))
  (if (Math|Boolean|ANDBoolean (== NewValue 0) (> OldValue 0) (not (Variables|/Script/BarnTowelRuntime.BarnTowelLayerAnimInstance|GetbThrusting (Variables|Transient|GetPlayerAnimBP))))
    (bind _returnvalue (Audio|SpawnSoundAttached (Variables|Sound|GetOutOfFuelSound) (Variables|/Script/Engine.Character|GetMesh (Variables|Default|GetPlayerPawn))))
    (Audio|Parameter|SetBooleanParameter _returnvalue "islocal" (Pawn|IsLocallyViewed (Variables|Default|GetPlayerPawn)) :!class /Script/AudioExtensions.AudioParameterControllerInterface)
    (Variables|Transient|SetPreviousFuelValue (Variables|GetNewValue))
    (else
      (Variables|Transient|SetPreviousFuelValue (Variables|GetNewValue)))))

(fn SetCameraEnabled
  (params (TargetActor :type Actor%20Object%20Reference) (Enabled :type Boolean))
  (Variables|Default|GetFortPlayerCameraBase
    (:then
      _)
    (:else
      (bind _asfort_player_pawn (Utilities|Casting|CastToFortPlayerPawn TargetActor))
      (LocalVariables|SetFortPlayerPawn _asfort_player_pawn)
      (bind _asfort_player_controller (Utilities|Casting|CastToFortPlayerController (Pawn|GetLocalViewingPlayerController _asfort_player_pawn)))
      (Variables|Default|SetFort%20Player%20Controller _asfort_player_controller)
      (bind _asfort_player_camera_base (Utilities|Casting|CastToFortPlayerCameraBase (Variables|/Script/Engine.PlayerController|GetPlayerCameraManager _asfort_player_controller)))
      (Variables|Default|SetFortPlayerCameraBase _asfort_player_camera_base)))
  (Utilities|FlowControl|Sequence
    (:then_0
      ;[1 1 1] Camera Mode override
      (if (Variables|GetEnabled)
        (if (not (GameplayTags|HasAnyMatchingGameplayTags (Variables|Default|GetPlayerPawn) (Variables|Camera|GetTags_PreventApplyFlyingCameraMode)))
          (Camera|AddCameraModeOverride (Variables|Default|GetFortPlayerCameraBase) (Variables|Camera|GetCameraModeClass_Flying) (Variables|Camera|GetTag_FlyingCameraMode) 0 0 0))
        (else
          (Camera|RemoveCameraModeOverride (Variables|Default|GetFortPlayerCameraBase) (Variables|Camera|GetTag_FlyingCameraMode) 0 false))))
    (:then_1
      (if (Variables|GetEnabled)
        ;[1 1 1] Enable Camera Modifier
        (bind _returnvalue (Game|Player|FindCameraModifierbyClass (Variables|Default|GetFortPlayerCameraBase) (Variables|Camera|GetCameraModifierClass_Flying)))
        (Utilities|IsValid _returnvalue
          (:Is%20Valid
            (CameraModifier|EnableModifier _returnvalue)
            (bind _ascamera_modifier_barn_towel (Utilities|Casting|CastToCameraModifier_BarnTowel _returnvalue))
            (Class|CameraModifierBarnTowel|Activate _ascamera_modifier_barn_towel))
          (:Is%20Not%20Valid
            (Game|Player|AddNewCameraModifier (Variables|Default|GetFortPlayerCameraBase) (Variables|Camera|GetCameraModifierClass_Flying))))
        (else
          ;[1 1 1] Disable Camera Modifier
          (bind _returnvalue_2 (Game|Player|FindCameraModifierbyClass (Variables|Default|GetFortPlayerCameraBase) (Variables|Camera|GetCameraModifierClass_Flying)))
          (Utilities|IsValid _returnvalue_2
            (:Is%20Valid
              (CameraModifier|DisableModifier _returnvalue_2 false))))))))

(fn SetFXFloatParameter
  (params (Name :type Name) (Float :type Float%20%28double-precision%29))
  (for _array_element (Variables|TransientComponents|GetFlyingTrailFX)
    (Effects|Components|ParticleSystem|SetFloatParameter _array_element (Variables|GetName) (Variables|GetFloat)))
  (for _array_element_2 (Variables|TransientComponents|GetAudioComponents)
    (Audio|Parameter|SetFloatParameter _array_element_2 (Variables|GetName) (Variables|GetFloat) :!class /Script/AudioExtensions.AudioParameterControllerInterface)))

(fn UpdateFXandAudioParams
  (bind _flyingtrailfx (Variables|TransientComponents|GetFlyingTrailFX))
  (bind _playeranimbp (Variables|Transient|GetPlayerAnimBP))
  (Variables|Transient|GetPlayerAnimBP
    (:then
      _)
    (:else
      (bind _asbarn_towel_player_anim_bp (Utilities|Casting|CastToBarnTowel_Player_AnimBP (Components|SkeletalMesh|AnimationBlueprintLinking|GetLinkedAnimLayerInstancebyClass (Variables|/Script/Engine.Character|GetMesh (Variables|Default|GetPlayerPawn)) "/BarnTowelGameplay/Assets/Animation/BarnTowel_Player_AnimBP.BarnTowel_Player_AnimBP_C")))
      (Variables|Transient|SetPlayerAnimBP _asbarn_towel_player_anim_bp)))
  (if (Utilities|Array|IsEmpty _flyingtrailfx)
    (LocalVariables|SetFlyingTrailLocalEffect 0)
    (else
      (LocalVariables|SetFlyingTrailLocalEffect (Utilities|Array|Get(acopy) _flyingtrailfx))))
  (Variables|SpeedSmoothing|SetPreviousSpeed (CallFunction|GetClampedSpeed (Variables|/Script/Engine.Actor|GetRootComponent (Variables|Default|GetPlayerPawn))))
  (CallFunction|SetFlyingParameters (LocalVariables|GetFlyingTrailLocalEffect) (Variables|TransientComponents|GetAudioComponents) (Variables|Default|GetPlayerPawn) 0.0 (Variables|/Script/BarnTowelRuntime.BarnTowelLayerAnimInstance|GetGlideRollCombined (Variables|Transient|GetPlayerAnimBP)) (- 1.0 (Variables|/Script/BarnTowelRuntime.BarnTowelLayerAnimInstance|GetRollAlpha (Variables|Transient|GetPlayerAnimBP))) (Variables|/Script/BarnTowelRuntime.BarnTowelLayerAnimInstance|GetGlidePitchAlpha (Variables|Transient|GetPlayerAnimBP)) false (Variables|/Script/BarnTowelRuntime.BarnTowelLayerAnimInstance|GetbIsAscending (Variables|Transient|GetPlayerAnimBP)) false (Variables|SpeedSmoothing|GetPreviousSpeed) (Variables|/Script/BarnTowelRuntime.BarnTowelLayerAnimInstance|GetbIsCrashing (Variables|Transient|GetPlayerAnimBP)) (Variables|/Script/BarnTowelRuntime.FortWeaponBarnTowel|GetNativeCurrentHeightAboveWorld (Variables|TransientComponents|GetBarnTowelWeapon)))
  (for _array_element (Variables|TransientComponents|GetAudioComponents)
    (Audio|Parameter|SetBooleanParameter _array_element "WindUp" (Variables|/Script/BarnTowelRuntime.BarnTowelLayerAnimInstance|GetbWindingUpThrust _playeranimbp) :!class /Script/AudioExtensions.AudioParameterControllerInterface)
    (Audio|Parameter|SetBooleanParameter _array_element "isHovering" (Variables|/Script/BarnTowelRuntime.BarnTowelLayerAnimInstance|GetbIsHovering _playeranimbp) :!class /Script/AudioExtensions.AudioParameterControllerInterface)))

(fn ConstructionScript)
