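; Looping gameplay cue for the BarnTowel take-off, adding a sound indicator and local-player VFX state to the parent cue's effects.
; On loop start it caches the target and, if valid, shows a local audio-visualization indicator at its location, passing the target's controller and ranged by the fire sound's max distance.
; It then sets IsLocalPlayer on the first particle component from whether the pawn is locally controlled.
; ShouldShowSoundIndicator returns false, while the affiliation override lists all three affiliations.
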(event EventOnLoopingStart(EitherVFX-preferthisversion)
  (params (MyTarget :type Actor%20Object%20Reference) (Parameters :type Gameplay%20Cue%20Parameters%20Structure%20%28by%20ref%29) (ParticleComponents :type Array%20of%20FXSystem%20Component%20Object%20References) (AudioComponents :type Array%20of%20Audio%20Component%20Object%20References))
  (|CallParentFunction MyTarget Parameters ParticleComponents AudioComponents)
  (bind _output_get (Variables|Default|SetCurrentTarget MyTarget))
  (Utilities|IsValid _output_get
    (:Is%20Valid
      ;[1 1 1] AV for blast (much larger distance)
      (Fortnite|ShowSoundIndicatorLocal 0 0 (Transformation|GetActorLocation _output_get) (Variables|/Script/Engine.SoundBase|GetMaxDistance (Variables|Default|GetFireSoundCue)) "Generic" (Utilities|Array|MakeArray (Fortnite|GetFortPlayerControllerfromActor _output_get)) 0 "(R=0.000000,G=0.000000,B=0.000000,A=0.000000)" "/BarnTowelGameplay/Assets/Icons/T_UI_AV_BarnTowel.T_UI_AV_BarnTowel" (Variables|Default|GetFireSoundCue))
      (bind _asplayer_pawn_athena (Utilities|Casting|CastToPlayerPawn_Athena (Variables|Default|GetCurrentTarget)))
      (Effects|Components|ParticleSystem|SetBoolParameter (Utilities|Array|Get(acopy) ParticleComponents) "IsLocalPlayer" (Pawn|IsLocallyControlled _asplayer_pawn_athena)))))

(fn GetAffiliationsToShowFor
  (return (Utilities|Array|MakeArray "Friendly" "Neutral" "Hostile")))

(fn ShouldShowSoundIndicator
  (params (PlayerController :type Fort%20Player%20Controller%20Object%20Reference))
  (return false))

(fn ConstructionScript)
