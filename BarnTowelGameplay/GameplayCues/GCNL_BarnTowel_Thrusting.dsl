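; Looping gameplay cue for thrusting with the fuel-based BarnTowel gadget, adding per-player VFX state and an out-of-fuel sound to the parent behavior.
; Loop start and recurring events set "islocal" and "isenemy" particle parameters from local control and team affiliation.
; Start and removal toggle "PPVEnable" on the second particle component; removal also plays an attached sound at zero fuel.
; Recurring and removal reuse some values from loop start, and sound indicators are disabled.
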(event EventOnLoopingStart(EitherVFX-preferthisversion)
  (params (MyTarget :type Actor%20Object%20Reference) (Parameters :type Gameplay%20Cue%20Parameters%20Structure%20%28by%20ref%29) (ParticleComponents :type Array%20of%20FXSystem%20Component%20Object%20References) (AudioComponents :type Array%20of%20Audio%20Component%20Object%20References))
  (bind _output (Utilities|Array|Get(acopy) ParticleComponents 1))
  (Effects|Components|ParticleSystem|SetBoolParameter _output "PPVEnable" true)
  (|CallParentFunction MyTarget Parameters ParticleComponents AudioComponents)
  (bind _asplayer_pawn_athena (Utilities|Casting|CastToPlayerPawn_Athena MyTarget))
  (Effects|Components|ParticleSystem|SetBoolParameter _output "islocal" (Pawn|IsLocallyControlled _asplayer_pawn_athena))
  (Effects|Components|ParticleSystem|SetBoolParameter _output "isenemy" (Utilities|Enum|Equal(Enum) (Team|GetTeamAffiliationforActortoLocalPlayer _asplayer_pawn_athena) "Hostile")))

(event EventOnRecurring(EitherVFX-preferthisone)
  (params (MyTarget :type Actor%20Object%20Reference) (Parameters :type Gameplay%20Cue%20Parameters%20Structure%20%28by%20ref%29) (ParticleComponents :type Array%20of%20FXSystem%20Component%20Object%20References) (AudioComponents :type Array%20of%20Audio%20Component%20Object%20References) (BurstCameraShakeInstance :type Camera%20Shake%20Base%20Object%20Reference) (BurstDecalInstance :type Decal%20Actor%20Object%20Reference))
  (bind _output (Utilities|Array|Get(acopy) ParticleComponents))
  (|CallParentFunction MyTarget Parameters ParticleComponents AudioComponents BurstCameraShakeInstance BurstDecalInstance)
  (bind _asplayer_pawn_athena (Utilities|Casting|CastToPlayerPawn_Athena MyTarget))
  (Effects|Components|ParticleSystem|SetBoolParameter _output "islocal" (Pawn|IsLocallyControlled _asplayer_pawn_athena))
  (Effects|Components|ParticleSystem|SetBoolParameter _output "isenemy" (Utilities|Enum|Equal(Enum) (Team|GetTeamAffiliationforActortoLocalPlayer (stale EventOnLoopingStart(EitherVFX-preferthisversion) _asplayer_pawn_athena)) "Hostile")))

(event EventOnRemoval(EitherVFX-preferthisversion)
  (params (MyTarget :type Actor%20Object%20Reference) (Parameters :type Gameplay%20Cue%20Parameters%20Structure%20%28by%20ref%29) (ParticleComponents :type Array%20of%20FXSystem%20Component%20Object%20References) (AudioComponents :type Array%20of%20Audio%20Component%20Object%20References) (BurstCameraShakeInstance :type Camera%20Shake%20Base%20Object%20Reference) (BurstDecalInstance :type Decal%20Actor%20Object%20Reference))
  (|CallParentFunction MyTarget Parameters ParticleComponents AudioComponents BurstCameraShakeInstance BurstDecalInstance)
  (Effects|Components|ParticleSystem|SetBoolParameter (Utilities|Array|Get(acopy) (stale EventOnLoopingStart(EitherVFX-preferthisversion) ParticleComponents) 1) "PPVEnable" false)
  (bind _asplayer_pawn_athena (Utilities|Casting|CastToPlayerPawn_Athena MyTarget))
  (if (== (GadgetAttributeSets|GetSimulatedAttributeValue _asplayer_pawn_athena (Variables|Default|GetFuelAttribute) 1.0 :!class /Script/FortniteGame.FortPlayerPawn) 0)
    (Audio|SpawnSoundAttached "/BarnTowelGameplay/Assets/Sounds/_various/BarnTowel_Fuel_Out_Cue.BarnTowel_Fuel_Out_Cue" (Variables|/Script/Engine.Character|GetMesh _asplayer_pawn_athena))))

(fn GetAffiliationsToShowFor
  (return (Utilities|Array|MakeArray "Friendly" "Neutral" "Hostile")))

(fn ShouldShowSoundIndicator
  (params (PlayerController :type Fort%20Player%20Controller%20Object%20Reference))
  (return false))

(fn ConstructionScript)
