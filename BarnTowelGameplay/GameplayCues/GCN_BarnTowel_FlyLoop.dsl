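; Looping gameplay cue notify for BarnTowel flight that drives particle effects and ground physics-field forces on the target.
; On start, it flags the first particle component for local players and toggles its smoke on when LocalFuel drops to 25 or below.
; A repeating timer traces downward, checks effects quality, and spawns FN_RadialForce at hits when physics fields are enabled.
; Removal destroys only the last stored force actor.
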(event Custom|CE_Tick
  (bind _returnvalue_location (Math|Transform|BreakTransform (Transformation|GetActorTransform (stale EventOnLoopingStart(EitherVFX-preferthisversion) MyTarget))))
  (bind (_outhit _returnvalue) (Collision|LineTraceForObjects 0 (+ _returnvalue_location "0.000000,0.000000,100.000000") (- _returnvalue_location "0.000000,0.000000,600.000000") (Variables|Default|GetObject%20Types)))
  (Utilities|EffectsQualityCheck 0
    (:Cinematic
      _)
    (:Epic
      _)
    (:High
      _))
  (if _returnvalue
    (bind _returnvalue_2 (Development|GetConsoleVariableIntValue "r.PhysicsField.EnableField"))
    (if (== _returnvalue_2 1)
      (bind _returnvalue_3 (Game|SpawnActorfromClass 0 "/Game/Effects/Niagara/Blueprints/PhsyicsFields/FN_RadialForce.FN_RadialForce_C" (Math|Transform|MakeTransform (.Location (Collision|BreakHitResult _outhit))) "Undefined" "MultiplyWithRoot" 0 0.2 0.8 5e+02 0.3 0.0 0))
      (Variables|Default|SetForce _returnvalue_3))))

(event EventOnLoopingStart(EitherVFX-preferthisversion)
  (params (MyTarget :type Actor%20Object%20Reference) (Parameters :type Gameplay%20Cue%20Parameters%20Structure%20%28by%20ref%29) (ParticleComponents :type Array%20of%20FXSystem%20Component%20Object%20References) (AudioComponents :type Array%20of%20Audio%20Component%20Object%20References))
  (bind _output (Utilities|Array|Get(acopy) ParticleComponents))
  (bind _asfort_player_pawn_athena (Utilities|Casting|CastToFortPlayerPawnAthena MyTarget))
  (Utilities|FlowControl|Sequence
    (:then_0
      (Utilities|Time|SetTimerbyEvent (delegate Custom|CE_Tick) 0.1 true))
    (:then_1
      (Effects|Components|ParticleSystem|SetBoolParameter _output "IsLocalPlayer" (Pawn|IsLocallyControlled _asfort_player_pawn_athena))
      (bind (_asynctaskproxy _attribute _newvalue) (Ability|Tasks|WaitForAttributeChanged _asfort_player_pawn_athena "(AttributeName=\"LocalFuel\",Attribute=/Script/FlyingCodeRuntime.FlyingAttributeSet:LocalFuel,AttributeOwner=\"/Script/CoreUObject.Class'/Script/FlyingCodeRuntime.FlyingAttributeSet'\")")
        (:Changed
          (if (> _newvalue 25.0)
            (Effects|Components|ParticleSystem|SetBoolParameter _output "TurnOnSmoke" false)
            (else
              (Effects|Components|ParticleSystem|SetBoolParameter _output "TurnOnSmoke" true))))))))

(event EventOnRemoval(EitherVFX-preferthisversion)
  (params (MyTarget :type Actor%20Object%20Reference) (Parameters :type Gameplay%20Cue%20Parameters%20Structure%20%28by%20ref%29) (ParticleComponents :type Array%20of%20FXSystem%20Component%20Object%20References) (AudioComponents :type Array%20of%20Audio%20Component%20Object%20References) (BurstCameraShakeInstance :type Camera%20Shake%20Base%20Object%20Reference) (BurstDecalInstance :type Decal%20Actor%20Object%20Reference))
  (Actor|DestroyActor (Variables|Default|GetForce)))

(fn ConstructionScript)
