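; Burst gameplay cue for the BarnTowel landing impact, handling its effects, audio and dive-bomb sound indicator.
; OnBurst optionally spawns a radial physics field (console-variable gated) and line-traces for water, swapping in splash FX and sound.
; It also flags locally viewed audio and stops listed sounds on the target.
; OnBurstGeneric aligns particles to the hit normal and, for non-character instigators, spawns a water splash or fading crack decal.
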(fn OnBurst
  (params (Target :type Actor%20Object%20Reference) (Parameters :type Gameplay%20Cue%20Parameters%20Structure%20%28by%20ref%29) (SpawnResults :type Gameplay%20Cue%20Notify%20Spawn%20Result%20Structure%20%28by%20ref%29))
  (bind (_asfort_pawn _bsuccess) (Utilities|Casting|CastToFortPawn (Variables|GetTarget)))
  (CallFunction|SpawnSoundIndicators Target)
  (Utilities|FlowControl|Sequence
    (:then_0
      (bind _returnvalue (Development|GetConsoleVariableIntValue "r.PhysicsField.EnableField"))
      (if (== _returnvalue 1)
        (Game|SpawnActorfromClass 0 "/Game/Effects/Niagara/Blueprints/PhsyicsFields/FN_RadialForce.FN_RadialForce_C" (Transformation|GetActorTransform Target) "Undefined" "MultiplyWithRoot" 0 1.0 0.75 1e+03 2.0 0.0 0)))
    (:then_1
      (bind _outhit (Collision|LineTraceForObjects 0 (+ (.location (Transformation|GetActorTransform Target)) "0.000000,0.000000,200.000000") (- (.location (Transformation|GetActorTransform Target)) "0.000000,0.000000,200.000000") (Variables|Default|GetObject%20Types)))
      (bind _output_get (LocalVariables|SetisWater (.bSuccess (Utilities|Casting|CastToWaterBody (.HitActor (Collision|BreakHitResult _outhit))))))
      (if _output_get
        (Components|Activation|Deactivate (Utilities|Array|Get(acopy) (Utilities|Struct|BreakGameplayCueNotifySpawnResult SpawnResults)))
        (Effects|Components|ParticleSystem|SpawnFXAtLocation :FXSystemAsset "/Game/Effects/Fort_Effects/GameplayPulginsAssets/Water/NS_Gameplay_LandImpactWater.NS_Gameplay_LandImpactWater" :Location (.Location (Collision|BreakHitResult _outhit)) :Scale "1.000000,1.000000,1.000000" :bAutoDestroy true :PoolingMethod "AutoRelease" :bAutoActivate true)
        (else
          _))
      (bind _returnvalue_2 (Audio|SpawnSoundatLocation 0 (select (LocalVariables|GetisWater) (Variables|Default|GetWaterImpact) (Variables|Default|GetExplosiveImpact)) (select (LocalVariables|GetisWater) (.Location (Collision|BreakHitResult _outhit)) (.location (Transformation|GetActorTransform (Variables|GetTarget))))))
      (Audio|Parameter|SetBooleanParameter _returnvalue_2 "islocal" (and (Pawn|IsLocallyViewed _asfort_pawn) _bsuccess) :!class /Script/AudioExtensions.AudioParameterControllerInterface))
    (:then_2
      (for _array_element (Actor|GetComponentsByClass (Variables|GetTarget) "/Script/Engine.AudioComponent")
        (Utilities|IsValid _array_element
          (:Is%20Valid
            (if (Utilities|Array|ContainsItem (Variables|Default|GetSounds%20to%20Stop) (Variables|/Script/Engine.AudioComponent|GetSound _array_element))
              (Audio|Components|Audio|Stop _array_element)))))
      (return))))

(fn OnBurstGeneric
  (params (MyTarget :type Actor%20Object%20Reference) (Parameters :type Gameplay%20Cue%20Parameters%20Structure%20%28by%20ref%29) (ParticleComponents :type Array%20of%20FXSystem%20Component%20Object%20References) (AudioComponents :type Array%20of%20Audio%20Component%20Object%20References) (BurstCameraShakeInstance :type Camera%20Shake%20Base%20Object%20Reference) (BurstDecalInstance :type Decal%20Actor%20Object%20Reference))
  (bind (_normalizedmagnitude _rawmagnitude _effectcontext _matchedtagname _originaltag _aggregatedsourcetags _aggregatedtargettags _location _normal _instigator) (Ability|GameplayCue|BreakGameplayCueParameters Parameters))
  (for _array_element ParticleComponents
    (Transformation|SetWorldLocation _array_element _location false false)
    (Transformation|SetWorldRotation _array_element (Math|Conversions|RotationFromXVector _normal) false false)
    (Utilities|Casting|CastToCharacter _instigator
      (:CastFailed
        (Utilities|Casting|CastToFortWaterBodyActor _instigator
          (:then
            (Effects|Components|ParticleSystem|SpawnFXAtLocation :WorldContextObject _instigator :FXSystemAsset "/Game/Effects/Fort_Effects/GameplayPulginsAssets/Water/NS_Gameplay_LandImpactWater.NS_Gameplay_LandImpactWater" :Location _location :Scale "1.000000,1.000000,1.000000" :bAutoDestroy true :PoolingMethod "AutoRelease" :bAutoActivate true))
          (:CastFailed
            (bind _returnvalue (Rendering|Decal|SpawnDecalatLocation 0 "/HighTower/Items/Tomato/Repulsors/FX/MD_Hightower_Tomato_Crack.MD_Hightower_Tomato_Crack" "200.000000,300.000000,300.000000" _location (Math|Conversions|RotationFromXVector _normal) 3.0))
            (Rendering|Components|Decal|SetFadeOut _returnvalue 3.0 2.0 true))))))
  (return))

(fn SpawnSoundIndicators
  (params (Target :type Actor%20Object%20Reference))
  (Utilities|IsValid Target
    (:Is%20Valid
      (Fortnite|ShowSoundIndicatorLocal 0 0 (Transformation|GetActorLocation Target) 0.0 "Gunshot" (Utilities|Array|MakeArray (Fortnite|GetFortPlayerControllerfromActor Target)) Target "(R=0.000000,G=0.000000,B=0.000000,A=0.000000)" "/FlightParentGameplay/UI/T_UI_AV_Common_DiveBomb_32x32.T_UI_AV_Common_DiveBomb_32x32"))))
