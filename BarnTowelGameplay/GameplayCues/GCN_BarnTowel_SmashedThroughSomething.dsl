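; Burst gameplay cue notify for BarnTowel that plays an impact sound at the target player pawn's location when it smashes through something.
; After calling the parent OnBurst, it truncates the normalized magnitude to select the hit-player cue (0) or hit-world cue (1).
; A comment names GA_Movement_RocketRan as the sender of that magnitude, which serves as a hit-type selector rather than an intensity.
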(fn OnBurst
  (params (Target :type Actor%20Object%20Reference) (Parameters :type Gameplay%20Cue%20Parameters%20Structure%20%28by%20ref%29) (SpawnResults :type Gameplay%20Cue%20Notify%20Spawn%20Result%20Structure%20%28by%20ref%29))
  ;[0 0 0] Choosing sound cue to play depending on value sent through normalized magnitude value (in GA_Movement_RocketRan). 0 = Player Pawn. 1 = World
  (|CallParentFunction Target Parameters SpawnResults)
  (bind _asplayer_pawn_athena (Utilities|Casting|CastToPlayerPawn_Athena Target))
  (Audio|PlaySoundatLocation 0 (Utilities|Select :Index (Math|Float|Truncate (Ability|GameplayCue|BreakGameplayCueParameters Parameters)) :Option%200 (Variables|Default|GetHitPlayerSoundCue) :Option%201 (Variables|Default|GetHitWorldSoundCue)) (Transformation|GetActorLocation _asplayer_pawn_athena))
  (return))
