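; Looping gameplay cue for the BarnTowel warm-up, adding a sound indicator and a short-lived physics field to the parent cue's effects.
; On loop start it shows an indicator at the target using the first burst sound, for all affiliations except the target's own controller.
; Where cosmetics can run, an effects quality check and the physics-field console variable gate spawning a FN_WashForce actor at the target.
; A retriggerable delay then destroys that force actor.
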(event EventOnLoopingStart(EitherVFX-preferthisversion)
  (params (MyTarget :type Actor%20Object%20Reference) (Parameters :type Gameplay%20Cue%20Parameters%20Structure%20%28by%20ref%29) (ParticleComponents :type Array%20of%20FXSystem%20Component%20Object%20References) (AudioComponents :type Array%20of%20Audio%20Component%20Object%20References))
  (|CallParentFunction MyTarget Parameters ParticleComponents AudioComponents)
  (Utilities|FlowControl|Sequence
    (:then_0
      (CallFunction|PlaySoundIndicator MyTarget))
    (:then_1
      (Networking|CanExecuteCosmeticEvents
        (:True
          (Utilities|EffectsQualityCheck 0
            (:Cinematic
              _)
            (:Epic
              _)
            (:High
              _))
          (bind _returnvalue (Development|GetConsoleVariableIntValue "r.PhysicsField.EnableField"))
          (if (== _returnvalue 1)
            (bind _returnvalue_2 (Game|SpawnActorfromClass 0 "/Game/Effects/Niagara/Blueprints/PhsyicsFields/FN_WashForce.FN_WashForce_C" (Transformation|GetActorTransform MyTarget) "Undefined" "MultiplyWithRoot" 0 1e+03 1.3 -999.0 0))
            (Variables|Default|SetForce _returnvalue_2)
            (Utilities|FlowControl|RetriggerableDelay 0 1.5)
            (bind _force (Variables|Default|GetForce)
              (:then
                (Actor|DestroyActor _force))
              (:else))))))))

(fn GetAffiliationsToShowFor
  (return (Utilities|Array|MakeArray "Friendly" "Neutral" "Hostile")))

(fn PlaySoundIndicator
  (params (CurrentTarget :type Actor%20Object%20Reference))
  (bind _applicationburst (Variables|ApplicationEffects|GetApplicationBurst))
  (bind (_burstparticles _burstsounds) (Utilities|Struct|BreakFortBurstEffectData _applicationburst))
  (bind (_condition _attachment _soundcue) (Utilities|Struct|BreakFortGameplayCueAudioInfo (Utilities|Array|Get(aref) _burstsounds)))
  (bind _output_get (Variables|Default|SetCurrentTarget CurrentTarget))
  (Utilities|IsValid _output_get
    (:Is%20Valid
      (if (> (Utilities|Array|Length _burstsounds) 0)
        (SoundIndicator|ShowSoundIndicator (Variables|Default|GetFortSoundIndicator) (Transformation|GetActorLocation (Variables|Default|GetCurrentTarget)) _soundcue 1.0 0)))))

(fn ShouldShowSoundIndicator
  (params (PlayerController :type Fort%20Player%20Controller%20Object%20Reference))
  (bind _currenttarget (Variables|Default|GetCurrentTarget))
  (Utilities|IsValid _currenttarget
    (:Is%20Valid
      (return (!= PlayerController (Fortnite|GetControllerfromActor _currenttarget))))
    (:Is%20Not%20Valid
      (return true))))

(fn ConstructionScript)
